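# Cinematic character material for the Midas Black Knight Mask 01 cosmetic gadget, built for skeletal meshes and mesh deformers.
# It uses material attributes, chaining shared functions: character PreFX, MF_SuperLevellingMask_CINE, MF_SimpleMaterial, then character PostFX.
# The super levelling step feeds its mask output into MF_SimpleMaterial, whose SkinFXMask is a constant one.
# PreFX AO bypasses the middle steps to drive PostFX's AO, Distance Blend and Fresnel inputs.
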Material /BRCosmetics/Gadgets/Assets/Midas_Black_Knight_Mask_01/Material/M_Midas_Black_Knight_Mask_01_Gadget_CINE
graph {
 properties {
  bUseMaterialAttributes=true
  bUsedWithMeshDeformer=true
  bUsedWithSkeletalMesh=true
  pixelDepthOffsetMode=PDOM_Legacy
  usedShadingModels=Default Lit
 }
 nodes [
  n1 = MaterialFunctionCall() {materialFunction=&BaseFN_CharacterFunction_PreFX}
  n2 = MaterialFunctionCall!(Local Normal, Local Position, Material=@n1.Result) {materialFunction=&/MF_SuperLevellingMask_CINE}
  MaterialFunctionCall(
   AO=@n1.AO
   Distance Blend=@n1.AO
   Fresnel=@n1.AO
   Material=MaterialFunctionCall!(SuperLevellingMask=@n2.SuperLevelling_Mask, SkinFXMask=Constant() {r=1}, Material=@n2.SuperLevelling_Mask) {materialFunction=/Game/Packages/Fortress_SharedMaterials/Base_Material_Functions/MF_SimpleMaterial}.Result
  ) {materialFunction=&BaseFN_CharacterFunction_PostFX}
 ]
}
